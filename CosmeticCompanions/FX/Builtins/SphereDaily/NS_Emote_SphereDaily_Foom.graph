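# Niagara effect for the SphereDaily companion emote that pops short-lived "Foom" and "Bam" sprites.
# Two burst emitters, Foom and Bam001, share a size curve that swells quickly, overshoots, then shrinks to zero.
# Both tint sprites by time of day through a color-select function.
# Bam001 has a fixed negative camera offset, while Foom's grows over life along a curve scaled separately for gameplay and front end.
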NiagaraSystem /CosmeticCompanions/FX/Builtins/SphereDaily/NS_Emote_SphereDaily_Foom
dependencies {usedDataInterfaces [&NiagaraDataInterfaceCurve], usedDynamicInputs [], usedModules [&SystemState]}
emitters {
 Bam001 {
  emitterUpdateScript [
   &EmitterState(Life Cycle Mode=::NewEnumerator1, Loop Behavior=::NewEnumerator1, Loop Duration=1, UseLoopDelay=false)
   &SpawnBurst_Instantaneous(Use Spawn Probability=false, UseLoopCountLimit=false)
  ]
  particleSpawnScript [
   &InitializeParticle(
    Color Mode=::NewEnumerator0
    Lifetime=0.9
    Sprite Size Mode=::NewEnumerator3
    Uniform Sprite Size=40
    UsePositionOffset=false
    Write Material Random=true
   )
   &CameraOffset(Camera Offset Amount=-5)
  ]
  particleUpdateScript [
   &ParticleState()
   &ScaleSpriteSize(Uniform Curve Sprite Scale=DataInterface() {propertyValues {Curve {keys [
        {interpMode=RCIM_Cubic, leaveTangent=21.8589, tangentMode=RCTM_Break, time=0, value=0}
        {arriveTangent=3.5, interpMode=RCIM_Cubic, leaveTangent=3.5, time=0.1308, value=1.4}
        {arriveTangent=-2.6024, interpMode=RCIM_Cubic, tangentMode=RCTM_Break, time=0.5, value=1.75}
        {arriveTangent=-0.0414, interpMode=RCIM_Cubic, leaveTangent=-1.3062, tangentMode=RCTM_Break, time=0.8, value=1}
        <at 0.9 0 RCIM_Cubic>
       ]}}})
   &ScaleColor(
    Scale RGB=&MakeVectorFromLinearColorRGB(Color=&FN_Select_Color_from_Time_Of_Day(Morning=<rgba 1 0.8463 0.7189 1>, Day=<rgba 1 1 1 1>, Evening=<rgba 1 0.8371 0.7936 1>, Night=<rgba 0.55 0.6089 1 1>))
    ScaleA=false
    ScaleRGB=true
   ) {module=ScaleColor001}
  ]
  rendererClasses [&NiagaraSpriteRendererProperties]
  renderers [{rendererClass=&NiagaraSpriteRendererProperties}]
 }
 Foom {
  emitterUpdateScript [
   &EmitterState(Life Cycle Mode=::NewEnumerator1, Loop Behavior=::NewEnumerator1, Loop Duration=1, UseLoopDelay=false)
   &SpawnBurst_Instantaneous(Use Spawn Probability=false, UseLoopCountLimit=false)
  ]
  particleSpawnScript [&InitializeParticle(
    Color Mode=::NewEnumerator0
    Lifetime=0.9
    Sprite Size Mode=::NewEnumerator3
    Uniform Sprite Size=30
    UsePositionOffset=false
    Write Material Random=true
   )]
  particleUpdateScript [
   &ParticleState()
   &ScaleSpriteSize(
    Uniform Curve Scale=1.4
    Uniform Curve Sprite Scale=DataInterface() {propertyValues {Curve {keys [
        {interpMode=RCIM_Cubic, leaveTangent=21.8589, tangentMode=RCTM_Break, time=0, value=0}
        {arriveTangent=3.5, interpMode=RCIM_Cubic, leaveTangent=3.5, time=0.1308, value=1.4}
        {arriveTangent=-2.6024, interpMode=RCIM_Cubic, tangentMode=RCTM_Break, time=0.5, value=1.75}
        {arriveTangent=-0.0414, interpMode=RCIM_Cubic, leaveTangent=-1.3062, tangentMode=RCTM_Break, time=0.8, value=1}
        <at 0.9 0 RCIM_Cubic>
       ]}}}
   )
   &ScaleColor(
    Scale RGB=&MakeVectorFromLinearColorRGB(Color=&FN_Select_Color_from_Time_Of_Day(Morning=<rgba 1 0.8463 0.7189 1>, Day=<rgba 1 1 1 1>, Evening=<rgba 1 0.8371 0.7936 1>, Night=<rgba 0.55 0.6089 1 1>))
    ScaleA=false
    ScaleRGB=true
   ) {module=ScaleColor001}
   &CameraOffset(Camera Offset Amount=&FloatFromCurve(
     FloatCurve=DataInterface() {propertyValues {Curve {defaultValue=1, keys [{tangentMode=RCTM_User, time=0, value=0.25}, {tangentMode=RCTM_User, time=1, value=5}]}}}
     Scale Curve=&FN_FrontEndFloatModifier(Gameplay=35, Front End=30)
    ))
  ]
  rendererClasses [&NiagaraSpriteRendererProperties]
  renderers [{rendererClass=&NiagaraSpriteRendererProperties}]
 }
}
summary {
 systemName=NS_Emote_SphereDaily_Foom
 emitters [
  {emitterName=Foom, rendererClasses [&NiagaraSpriteRendererProperties]}
  {emitterName=Bam001, rendererClasses [&NiagaraSpriteRendererProperties]}
 ]
 userVariables [
  User.bIsFrontEnd {type=NiagaraBool}
  User.bIsFrontEndPreview {type=NiagaraBool}
  User.IsFrontEnd {type=NiagaraBool, defaultValue=true}
  User.IsLargeBody {type=NiagaraBool}
  User.Skeletal Mesh {
   type=NiagaraDataInterfaceSkeletalMesh
   defaultValue {
    struct=NiagaraExt_VariableValue_DataInterface
    value {dataInterface=&self:NiagaraDataInterfaceSkeletalMesh_0}
   }
  }
 ]
}
